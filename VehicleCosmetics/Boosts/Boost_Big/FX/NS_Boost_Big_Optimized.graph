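# Optimized Niagara system for the Big vehicle boost cosmetic, drawing glowing pink ribbon trails behind the car.
# Two layered emitters spawn ribbons per filtered boost socket, tracking skeletal positions offset by User.DTCarOffset.
# Particles shoot backward while inherited vehicle movement decays, widening and fading over their short lifetime.
# The first emitter adds curve-driven glow width and secondary color across three renderers; its painted-variant override is disabled, as is the second's.
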NiagaraSystem /VehicleCosmetics/Boosts/Boost_Big/FX/NS_Boost_Big_Optimized
dependencies {
 usedDataInterfaces [&NiagaraDataInterfaceCurve, &NiagaraDataInterfaceArrayFloat, &NiagaraDataInterfaceSkeletalMesh]
 usedDynamicInputs []
 usedModules [
  &self:SystemSpawnScript.SystemScriptSource.SystemScriptGraph.NiagaraNodeAssignment_0.SetVariables_#1
  &SystemState
 ]
}
emitters {
 NE_VC_GlowRibbonBack_Optimized001 {
  emitterSpawnScript [/SetVariables(Emitter.GlowRibbonWidth=45) {isSetParameters=true, module=SetVariables_#2}]
  emitterUpdateScript [
   &EmitterState()
   &VC_FilteredSocketCount()
   &VC_RateQualityHandler(High Spawn Rate=30, Low Spawn Rate=30, Use Custom Rate?=true)
   &VC_SpawnRibbon_v1(Number of Ribbons=$Emitter.PortCount, Spawn Time=-1, SpawnRate=$Emitter.Spawn Rate)
  ]
  particleSpawnScript [
   &VC_PaintedVariantsOverride() {enabled=false}
   &VC_InitializeRibbon_v1(
    Custom Offset=true
    Enable Skeletal Position=true
    Lifetime=0.7
    Local Vector Offset=&AddVectorToPosition(Position=<95 0 0>, Vector=$User.DTCarOffset)
    Offset Amount=1
    Ribbon Width=100
   )
   &AddVelocity(Velocity=<-3000 0 0>) {module=AddVelocity001}
   &Color(Color=<rgba 2 0.6 1.58 1>)
   &ShapeLocation(MaskSpawnGroup=false, Override Seed=false, Sphere Radius=1, UseOwnerScale=true) {enabled=false}
   /VehicleCosmetics/FX/Niagara/Modules/MergeEmitters/VC_ExtraBindingColors(ExtraColors=1, R1 Color:LinearColor=<rgba 0.55 0.165 0.4594 2>)
  ]
  particleUpdateScript [
   &ParticleState()
   &InheritSourceMovement(
    Applied Position Scale=&VectorFromFloat(Value=&FloatFromCurve(FloatCurve=DataInterface() {propertyValues {Curve {keys [
          {interpMode=RCIM_Cubic, tangentMode=RCTM_Break, time=0, value=1}
          {arriveTangent=-2.9025, interpMode=RCIM_Cubic, tangentMode=RCTM_Break, time=1, value=0}
         ]}}}))
    ApplyForce=false
    ApplyPosition=true
    ApplyPositionFalloff=false
   )
   &SolveForcesAndVelocity(Clamp Velocity=false, Limit Acceleration=false)
   &VC_UpdateRibbon_v1(Enable Skeletal Position=true)
   &ScaleColor(
    Scale Alpha=&FloatFromCurve(FloatCurve=DataInterface() {propertyValues {Curve {keys [<at 0 1 RCIM_Cubic>,<at 1 0 RCIM_Cubic>]}}})
    ScaleA=true
    ScaleRGB=true
   )
   &DynamicMaterialParameters(
    Index 0 Param 1=-0.1
    Module.Index0Param0Write=true
    Module.Index0Param1Write=true
    Module.Index0Param2Write=true
    Module.Index0Param3Write=true
   )
   &ScaleRibbonWidth(Ribbon Width Scale=&FloatFromCurve(FloatCurve=DataInterface() {propertyValues {Curve {keys [<at 0 0.65 RCIM_Cubic>,<at 1 1 RCIM_Cubic>]}}}))
   /SetVariables(
    Particles.GlowRibbonWidth:NiagaraFloat=&Multiply_Float(
     A=$Emitter.GlowRibbonWidth
     B=&FloatFromCurve(
      FloatCurve=DataInterface() {propertyValues {Curve {keys [<at 0 0.65>,<at 1 1>]}}}
      Scale Curve=&SelectFloatFromArray(Direct Array Index=$Particles.SpawnGroup, Float Selection Array=DataInterface() {propertyValues {FloatData [2,2]}})
     )
    )
    Particles.Linear Color=&VC_DynamicScaleColor(
     Color to Scale:LinearColor=$Particles.R1 Color
     RGB Scale:Vector3f=&VectorFromFloat(Value=&FloatFromCurve(FloatCurve=DataInterface() {propertyValues {Curve {keys [<at 0 4>,<at 1 4>]}}}))
     Alpha Scale:NiagaraFloat=&FloatFromCurve(FloatCurve=DataInterface() {propertyValues {Curve {keys [{tangentMode=RCTM_User, time=0, value=0}, {tangentMode=RCTM_User, time=1, value=1}]}}})
    )
   ) {isSetParameters=true}
  ]
  rendererClasses [&NiagaraRibbonRendererProperties]
  renderers [
   {rendererClass=&NiagaraRibbonRendererProperties}
   {rendererClass=&NiagaraRibbonRendererProperties, rendererIndex=1}
   {rendererClass=&NiagaraRibbonRendererProperties, rendererIndex=2}
  ]
 }
 NE_VC_GlowRibbonBack_Optimized002 {
  emitterSpawnScript [/SetVariables(Emitter.GlowRibbonWidth=51) {isSetParameters=true, module=SetVariables_#2}]
  emitterUpdateScript [
   &EmitterState()
   &VC_FilteredSocketCount()
   &VC_RateQualityHandler(High Spawn Rate=30, Low Spawn Rate=30, Use Custom Rate?=true)
   &VC_SpawnRibbon_v1(Number of Ribbons=$Emitter.PortCount, Spawn Time=-1, SpawnRate=$Emitter.Spawn Rate)
  ]
  particleSpawnScript [
   &VC_PaintedVariantsOverride() {enabled=false}
   &VC_InitializeRibbon_v1(
    Custom Offset=true
    Enable Skeletal Position=true
    Lifetime=0.75
    Local Vector Offset=&AddVectorToPosition(Position=<95 0 0>, Vector=$User.DTCarOffset)
    Offset Amount=1
    Ribbon Width=130
   )
   &AddVelocity(Velocity=<-3000 0 0>) {module=AddVelocity001}
   &Color(Color=<rgba 1 0 0.875 7>)
   &VC_ExtraBindingColors(ExtraColors=1, R1 Color:LinearColor=<rgba 0.55 0.165 0.4594 2>)
  ]
  particleUpdateScript [
   &ParticleState()
   &InheritSourceMovement(
    Applied Position Scale=&VectorFromFloat(Value=&FloatFromCurve(FloatCurve=DataInterface() {propertyValues {Curve {keys [
          {interpMode=RCIM_Cubic, tangentMode=RCTM_Break, time=0, value=1}
          {arriveTangent=-2.9025, interpMode=RCIM_Cubic, tangentMode=RCTM_Break, time=1, value=0}
         ]}}}))
    ApplyForce=false
    ApplyPosition=true
    ApplyPositionFalloff=false
   )
   &SolveForcesAndVelocity(Clamp Velocity=false, Limit Acceleration=false)
   &VC_UpdateRibbon_v1(Enable Skeletal Position=true)
   &ScaleColor(
    Scale Alpha=&FloatFromCurve(FloatCurve=DataInterface() {propertyValues {Curve {keys [<at 0 1 RCIM_Cubic>,<at 1 0 RCIM_Cubic>]}}})
    ScaleA=true
    ScaleRGB=true
   )
   &DynamicMaterialParameters(
    Index 0 Param 1=-0.1
    Module.Index0Param0Write=true
    Module.Index0Param1Write=true
    Module.Index0Param2Write=true
    Module.Index0Param3Write=true
   )
   &ScaleRibbonWidth(Ribbon Width Scale=&FloatFromCurve(FloatCurve=DataInterface() {propertyValues {Curve {keys [<at 0 0.65 RCIM_Cubic>,<at 1 1 RCIM_Cubic>]}}}))
  ]
  rendererClasses [&NiagaraRibbonRendererProperties]
  renderers [
   {rendererClass=&NiagaraRibbonRendererProperties}
   {rendererClass=&NiagaraRibbonRendererProperties, rendererIndex=1}
  ]
 }
}
summary {
 systemName=NS_Boost_Big_Optimized
 emitters [
  {emitterName=NE_VC_GlowRibbonBack_Optimized001, rendererClasses [&NiagaraRibbonRendererProperties]}
  {emitterName=NE_VC_GlowRibbonBack_Optimized002, rendererClasses [&NiagaraRibbonRendererProperties]}
 ]
 userVariables [
  User.Alpha_line {type=NiagaraFloat, defaultValue=1, description=""}
  User.bIsFrontendPreview {type=NiagaraBool}
  User.bIsPainted {type=NiagaraBool}
  User.black {type=NiagaraFloat, defaultValue=1.5, description=""}
  User.ColorVariantID {type=NiagaraInt32}
  User.DTCarOffset {type=Vector3f, description="", defaultValue {struct=Vector3f, value=<0 0 0>}}
  User.Intensity {type=NiagaraFloat, defaultValue=1, description=""}
  User.PaintedColor {type=LinearColor, description="", defaultValue {struct=LinearColor, value=<rgba 1 1 1 1>}}
  User.PaintedScalar {type=NiagaraFloat, defaultValue=1, description=""}
  User.PortCount {type=NiagaraInt32, defaultValue=2}
 ]
}
